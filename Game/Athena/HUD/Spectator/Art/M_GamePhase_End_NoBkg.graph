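# UI material for the spectator HUD, named as the end game-phase icon without a background.
# It is a translucent User Interface domain material whose graph holds a single texture sample of the 64-pixel check icon.
# The listing shows no parameters, math or connections from the sample to the material outputs.
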Material /Game/Athena/HUD/Spectator/Art/M_GamePhase_End_NoBkg
graph {
 properties {blendMode=BLEND_Translucent, materialDomain=MD_UI, pixelDepthOffsetMode=PDOM_Legacy, usedShadingModels=Default Lit}
 nodes [TextureSample() {texture=/Game/UI/Foundation/Textures/Icons/Checks/T-Icon-Check-64}]
}
