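# Translucent UI material that draws the Jump game-phase icon for the Athena spectator HUD, with no background layer.
# A scalar parameter multiplies the texture coordinates, then a half-unit offset minus a fraction of that scale re-centers them.
# The resulting UVs sample T_Jump, so the default scale below one enlarges the icon around its middle.
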Material /Game/Athena/HUD/Spectator/Art/M_GamePhase_Jump_NoBkg
graph {
 properties {blendMode=BLEND_Translucent, materialDomain=MD_UI, pixelDepthOffsetMode=PDOM_Legacy, usedShadingModels=Default Lit}
 nodes [
  n1 = ScalarParameter() {defaultValue=0.7}
  n2 = Multiply(A=TextureCoordinate(), B=@n1)
  TextureSample(UVs=Subtract(A=Add(A=@n2) {constB=0.5}, B=Divide(A=@n1))) {texture=/Game/Athena/HUD/Phase/T_Jump}
 ]
}
