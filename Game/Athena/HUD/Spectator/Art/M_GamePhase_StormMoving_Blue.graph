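# UI material for the spectator HUD that shows a blue, pulsing storm-moving game phase icon.
# A looping, pause-independent Time value drives a BellCurve function to pulse the brightness of a blue tint on the phase background texture.
# The storm texture blends over that tinted background using its own RGB as the lerp alpha.
# Its alpha is added to the slightly dimmed background alpha, presumably for opacity.
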Material /Game/Athena/HUD/Spectator/Art/M_GamePhase_StormMoving_Blue
graph {
 properties {blendMode=BLEND_Translucent, materialDomain=MD_UI, pixelDepthOffsetMode=PDOM_Legacy, usedShadingModels=Default Lit}
 nodes [
  n1 = TextureSample() {texture=/Game/Athena/HUD/Phase/T_StormMoving01}
  n2 = TextureSample() {texture=/Game/Athena/HUD/Phase/T_GamePhaseBackground}
  Add(A=@n1.A, B=Multiply(A=@n2.A) {constB=0.8})
  n3 = Multiply(A=Time() {bIgnorePause=true, bOverride_Period=true, period=1}) {constB=2}
  n4 = LinearInterpolate(Alpha=MaterialFunctionCall!(x=Subtract(A=@n3)) {materialFunction=/Engine/Functions/Engine_MaterialFunctions01/Gradient/BellCurve}.0-1) {constA=0.5, constB=2}
  n5 = Constant3Vector() {constant=<rgba 0.0595 0.2542 0.8963 1>}
  LinearInterpolate(A=Multiply(A=Multiply(A=@n4, B=@n5), B=@n2.RGB), B=@n1.RGB, Alpha=@n1.RGB)
 ]
}
