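# Spectator broadcast HUD widget pairing a game-phase icon with the remaining-time text inherited from AthenaBroadcastGamePhaseWidget.
# OnGamePhaseStepChanged calls SetIcon, which maps each non-None phase step to a material on the Icon image, with several steps sharing materials.
# PreConstruct shows or collapses the icon from bShowIcon and, at design time only, previews DesignTime_GamePhaseStep.
# The graph sets only the icon, never the time text.
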WidgetBlueprint /FortniteCoreUI/Athena/HUD/Spectator/MatchStatus/AthenaBroadcastGameTimeWidget
graphs [
 EventGraph {nodes [
   n1 = Event OnGamePhaseStepChanged!() -> (OutputDelegate, then, GamePhaseStep:EAthenaGamePhaseStep Enum=None) {is=title K2Node_Event}
   > SetIcon!(execute, self, GamePhaseStep:EAthenaGamePhaseStep Enum=@n1.GamePhaseStep) -> (then)
   n2 = Event PreConstruct()
   > SetVisibility(
    self=Get Icon() -> (Icon:Image Object Reference)
    InVisibility=Select(Option 0=Collapsed, Option 1=SelfHitTestInvisible, Index=Get bShowIcon() -> (bShowIcon:bool))
   )
   > Branch(Condition=@n2.IsDesignTime)
   > SetIcon(GamePhaseStep=Get DesignTime_GamePhaseStep() -> (DesignTime_GamePhaseStep:EAthenaGamePhaseStep Enum=None))
  ]}
 SetIcon {nodes [
   n1 = SetIcon!() -> (then, GamePhaseStep:EAthenaGamePhaseStep Enum=None) {is=title K2Node_FunctionEntry}
   > Branch(Condition="Not Equal (Enum)"!(A:EAthenaGamePhaseStep Enum=@n1, B:EAthenaGamePhaseStep Enum=None) -> (ReturnValue:bool))
   > SetBrushFromMaterial(
    self=Get Icon()
    Material=Select!(
     None:Material Interface Object Reference
     Setup:Material Interface Object Reference=/Game/Athena/HUD/Spectator/Art/M_GamePhase_End_NoBkg
     Warmup:Material Interface Object Reference=/Game/Athena/HUD/Spectator/Art/M_GamePhase_Warmup_NoBkg
     GetReady:Material Interface Object Reference=/Game/Athena/HUD/Spectator/Art/M_GamePhase_Bus_NoBkg
     BusLocked:Material Interface Object Reference=&M_GamePhase_Bus_NoBkg
     BusFlying:Material Interface Object Reference=/Game/Athena/HUD/Spectator/Art/M_GamePhase_Jump_NoBkg
     StormForming:Material Interface Object Reference=&M_GamePhase_Warmup_NoBkg
     StormHolding:Material Interface Object Reference=&M_GamePhase_Warmup_NoBkg
     StormShrinking:Material Interface Object Reference=/Game/Athena/HUD/Spectator/Art/M_GamePhase_StormMoving_Blue
     Countdown:Material Interface Object Reference=&M_GamePhase_Warmup_NoBkg
     FinalCountdown:Material Interface Object Reference=&M_GamePhase_Warmup_NoBkg
     EndGame:Material Interface Object Reference=&M_GamePhase_End_NoBkg
     Count:Material Interface Object Reference
     Index:EAthenaGamePhaseStep Enum=@n1
    ) -> (ReturnValue:Material Interface Object Reference)
   )
  ]}
]
variables [
 bShowIcon {type=boolean, default=true}
 DesignTime_GamePhaseStep {type=EAthenaGamePhaseStep, default=EndGame}
]
widgets {
 info {
  inheritedWidgetCount=2
  parentClass=/Script/FortniteUI.AthenaBroadcastGamePhaseWidget
  rootWidgetClass=&HorizontalBox
  widgetCount=4
 }
 widgets [
  HorizontalBox_0 {widgetClassPath=&HorizontalBox, widget=HorizontalBox_0}
  Icon {
   widgetClassPath=&Image
   bIsVariable=true
   parent=HorizontalBox_0
   widget=Icon
   properties {
    visibility=SelfHitTestInvisible
    brush {
     bIsDynamicallyLoaded=false
     drawAs=Image
     imageType=NoImage
     mirroring=NoMirror
     resourceName=None
     resourceObject=&M_GamePhase_End_NoBkg
     tiling=NoTile
     imageSize=<36 36>
     margin=<ltrb 0 0 0 0>
     outlineSettings {
      bUseBrushTransparency=false
      roundingType=HalfHeightRadius
      width=0
      color {colorUseRule=UseColor_Specified, specifiedColor=<rgba 0 0 0 0>}
      cornerRadii=<0 0 0 0>
     }
     tintColor {colorUseRule=UseColor_Specified, specifiedColor=<rgba 1 1 1 1>}
     uVRegion {bIsValid=false, max=<0 0>, min=<0 0>}
    }
   }
   slot {class=HorizontalBoxSlot, properties {horizontalAlignment=HAlign_Left, verticalAlignment=VAlign_Center, padding=<ltrb 0 0 10 0>}}
  }
  Text_RemainingTime {
   widgetClassPath=&TextBlock
   bInherited=true
   parent=HorizontalBox_0
   widget=Text_RemainingTime
   properties {
    bSimpleTextMode=true
    justification=Center
    text=0:00
    visibility=SelfHitTestInvisible
    font {
     bForceMonospaced=false
     bMaterialIsStencil=false
     fontMaterial=null
     fontObject=&BurbankSmall
     letterSpacing=0
     monospacedWidth=1
     size=28
     skewAmount=0
     typefaceFontName=Black
     outlineSettings {
      bApplyOutlineToDropShadows=false
      bMiteredCorners=false
      bSeparateFillAlpha=false
      outlineMaterial=null
      outlineSize=0
      outlineColor=<rgba 0 0 0 1>
     }
    }
   }
   slot {class=HorizontalBoxSlot, properties {verticalAlignment=VAlign_Center}}
  }
  TutorialHighlight {widgetClassPath=&TutorialHighlight}
 ]
}
